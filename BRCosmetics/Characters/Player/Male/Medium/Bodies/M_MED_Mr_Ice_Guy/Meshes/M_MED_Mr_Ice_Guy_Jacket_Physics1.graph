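# Physics asset for the coattails of the Mr Ice Guy medium male player character's jacket.
# Box bodies cover the front, side and back dynamic coattail panels, while capsules cover pelvis, lower spine, thighs and small anchor bones.
# Panels use limited swing and little or no twist, hanging asymmetrically from base bones, a left master bone or spine_01.
# The front-right panel also constrains to the right thigh; body and base-bone links swing freely.
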PhysicsAsset /BRCosmetics/Characters/Player/Male/Medium/Bodies/M_MED_Mr_Ice_Guy/Meshes/M_MED_Mr_Ice_Guy_Jacket_Physics1
bodies {
 base_coattail_back_l [{shapeType=Capsule, length=0.505, radius=0.505, shapeName=base_coattail_back_l_capsule, rotation=<rot 0 -90 0>}]
 base_coattail_front_r [{shapeType=Capsule, length=0.505, radius=0.505, shapeName=base_coattail_front_r_capsule, rotation=<rot 0 -90 0>}]
 base_coattail_l [{shapeType=Capsule, length=0.505, radius=0.505, shapeName=base_coattail_l_capsule, rotation=<rot 0 -90 0>}]
 base_master_coat [{shapeType=Capsule, length=0.505, radius=0.505, shapeName=base_master_coat_capsule, rotation=<rot 0 -90 0>}]
 dyn_coattail_back_l [{shapeType=Box, extentX=17.5, extentY=4, extentZ=16, center=<4.9835 -0.6968 -1.0419>, rotation=<rot 0 0 -10.0014>}]
 dyn_coattail_back_r [{shapeType=Box, extentX=17.5, extentY=4, extentZ=16, center=<4.8999 -1.4231 -0.8535>, rotation=<rot -1.2247 7.0145 9.856>}]
 dyn_coattail_front_l [{shapeType=Box, extentX=18.5, extentY=3, extentZ=17.5303, center=<9.3147 0.2531 0>}]
 dyn_coattail_front_r [{shapeType=Box, extentX=21.6227, extentY=1.0679, extentZ=17.4859, center=<7.0246 0.3881 -1.0215>}]
 dyn_coattail_l [{shapeType=Box, extentX=16.5, extentY=3.5, extentZ=15, center=<7.9994 -1.1908 -0.2089>, rotation=<rot 0.0489 0.2773 -10.0032>}]
 dyn_coattail_r [{shapeType=Box, extentX=18.6716, extentY=3.5, extentZ=17.6955, center=<7.0793 -1.1213 0.0004>}]
 dyn_master_coattail_l [{shapeType=Capsule, length=1, radius=1}]
 pelvis [{shapeType=Capsule, length=8.25, radius=8.7251, center=<-0.4197 -0.7044 -0.3572>}]
 spine_01 [{shapeType=Capsule, length=0, radius=0.5, center=<6.0433 2.3882 0.0946>}]
 thigh_l [
  {shapeType=Capsule, length=5, radius=2, center=<-2.6615 9.2324 0.0762>, rotation=<rot 0 0 24.9969>}
  {
   shapeType=Capsule
   length=19.9798
   radius=7.4141
   shapeName=thigh_l_capsule
   center=<-19.0866 2.9684 -0.8819>
   rotation=<rot -87.2277 -0.0012 8.3016>
  }
 ]
 thigh_r [{shapeType=Capsule, length=20, radius=7.5, center=<17.7353 -2.2587 0.0006>, rotation=<rot 90 7.7958 0>}]
}
constraints [
 {bone1Name=thigh_r, bone2Name=dyn_coattail_front_r, swing1LimitDegrees=45, swing2LimitDegrees=45, twistLimitDegrees=15}
 {bone1Name=dyn_coattail_back_r, bone2Name=spine_01, swing1LimitDegrees=10, swing2LimitDegrees=10, twistLimitDegrees=0}
 {bone1Name=dyn_coattail_r, bone2Name=spine_01, swing1LimitDegrees=30, swing2LimitDegrees=5, twistLimitDegrees=0}
 {bone1Name=dyn_coattail_front_r, bone2Name=base_coattail_front_r, swing1LimitDegrees=30, swing2LimitDegrees=5, twistLimitDegrees=0}
 {
  bone1Name=thigh_l
  bone2Name=pelvis
  swing1LimitDegrees=45
  swing1Motion=Free
  swing2LimitDegrees=45
  swing2Motion=Free
  twistLimitDegrees=45
  twistMotion=Free
 }
 {
  bone1Name=thigh_r
  bone2Name=pelvis
  swing1LimitDegrees=45
  swing1Motion=Free
  swing2LimitDegrees=45
  swing2Motion=Free
  twistLimitDegrees=45
  twistMotion=Free
 }
 {
  bone1Name=spine_01
  bone2Name=pelvis
  swing1LimitDegrees=45
  swing1Motion=Free
  swing2LimitDegrees=45
  swing2Motion=Free
  twistLimitDegrees=45
  twistMotion=Free
 }
 {bone1Name=dyn_coattail_l, bone2Name=base_coattail_l, swing1LimitDegrees=30, swing2LimitDegrees=5, twistLimitDegrees=0}
 {bone1Name=dyn_coattail_back_l, bone2Name=base_coattail_back_l, swing1LimitDegrees=45, swing2LimitDegrees=45, twistLimitDegrees=0}
 {bone1Name=dyn_master_coattail_l, bone2Name=spine_01, swing1LimitDegrees=45, swing2LimitDegrees=45, twistLimitDegrees=45}
 {bone1Name=dyn_coattail_front_l, bone2Name=dyn_master_coattail_l, swing1LimitDegrees=20, swing2LimitDegrees=5, twistLimitDegrees=0}
 {
  bone1Name=base_coattail_back_l
  bone2Name=spine_01
  swing1LimitDegrees=45
  swing1Motion=Free
  swing2LimitDegrees=45
  swing2Motion=Free
  twistLimitDegrees=45
  twistMotion=Free
 }
 {
  bone1Name=base_coattail_front_r
  bone2Name=spine_01
  swing1LimitDegrees=45
  swing1Motion=Free
  swing2LimitDegrees=45
  swing2Motion=Free
  twistLimitDegrees=45
  twistMotion=Free
 }
 {
  bone1Name=base_coattail_l
  bone2Name=spine_01
  swing1LimitDegrees=45
  swing1Motion=Free
  swing2LimitDegrees=45
  swing2Motion=Free
  twistLimitDegrees=45
  twistMotion=Free
 }
 {
  bone1Name=base_master_coat
  bone2Name=spine_01
  swing1LimitDegrees=45
  swing1Motion=Free
  swing2LimitDegrees=45
  swing2Motion=Free
  twistLimitDegrees=45
  twistMotion=Free
 }
]
