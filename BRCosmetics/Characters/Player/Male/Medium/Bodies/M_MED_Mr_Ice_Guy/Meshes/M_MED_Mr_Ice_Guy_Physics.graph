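# Physics asset for the Mr Ice Guy medium male player body cosmetic, covering its dangling accessories rather than a full ragdoll.
# Capsules, plus a box for the pickaxe, cover the dyn_ accessory bones and their calf, thigh and spine anchors.
# Shoelaces, collar strings, pickaxe, canister and satchel constrain to those anchors, while two straps chain off the satchel.
# Every constraint locks twist; pickaxe and straps also lock one swing axis for hinge-like motion.
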PhysicsAsset /BRCosmetics/Characters/Player/Male/Medium/Bodies/M_MED_Mr_Ice_Guy/Meshes/M_MED_Mr_Ice_Guy_Physics
bodies {
 calf_twist_02_l [{shapeType=Capsule, length=6.25, radius=2, rotation=<rot 90 0 0>}]
 calf_twist_02_r [{shapeType=Capsule, length=6.25, radius=2, rotation=<rot 90 0 0>}]
 dyn_canister_r [{shapeType=Capsule, length=10, radius=4, center=<0 0 1>, rotation=<rot 90 0 0>}]
 dyn_collarstring_l [{shapeType=Capsule, length=6, radius=1, center=<3 0 0>, rotation=<rot -90 0 25>}]
 dyn_collarstring_r [{shapeType=Capsule, length=6, radius=1, center=<3 0 0>, rotation=<rot -90 0 25>}]
 dyn_pickaxe [{shapeType=Box, extentX=20, extentY=4, extentZ=20, center=<9 0 0>, rotation=<rot 0 0 -30>}]
 dyn_satchel_l [{shapeType=Capsule, length=10, radius=3.75, rotation=<rot 0 0 70>}]
 dyn_shoelace_l [{shapeType=Capsule, length=2, radius=1, center=<1.5 0 0>, rotation=<rot 90 0 0>}]
 dyn_shoelace_r [{shapeType=Capsule, length=2, radius=1, center=<1.5 0 0>, rotation=<rot 90 0 0>}]
 dyn_strap_01 [{shapeType=Capsule, length=3.7351, radius=1.1934, center=<3.0307 0.8646 -0.1156>, rotation=<rot 0 -74.1417 -86.9717>}]
 dyn_strap_02 [{shapeType=Capsule, length=4.0245, radius=1.0928, center=<3.7046 0.3903 0.1813>, rotation=<rot 0 99.8438 -88.0408>}]
 spine_01 [{shapeType=Capsule, length=10, radius=9}]
 spine_05 [{shapeType=Capsule, length=8, radius=8, center=<1.6519 7.5523 0>}]
 thigh_l [{shapeType=Capsule, length=10, radius=7.75, center=<-15 0 0>, rotation=<rot 90 0 0>}]
}
constraints [
 {bone1Name=dyn_shoelace_r, bone2Name=calf_twist_02_r, swing1LimitDegrees=20, swing2LimitDegrees=5, twistLimitDegrees=45, twistMotion=Locked}
 {bone1Name=dyn_shoelace_l, bone2Name=calf_twist_02_l, swing1LimitDegrees=20, swing2LimitDegrees=5, twistLimitDegrees=45, twistMotion=Locked}
 {bone1Name=dyn_collarstring_l, bone2Name=spine_05, swing1LimitDegrees=20, swing2LimitDegrees=5, twistLimitDegrees=45, twistMotion=Locked}
 {bone1Name=dyn_collarstring_r, bone2Name=spine_05, swing1LimitDegrees=20, swing2LimitDegrees=5, twistLimitDegrees=45, twistMotion=Locked}
 {
  bone1Name=dyn_pickaxe
  bone2Name=thigh_l
  swing1LimitDegrees=30
  swing2LimitDegrees=45
  swing2Motion=Locked
  twistLimitDegrees=2
  twistMotion=Locked
 }
 {
  bone1Name=dyn_satchel_l
  bone2Name=spine_01
  swing1LimitDegrees=100
  swing1Motion=Free
  swing2LimitDegrees=10
  twistLimitDegrees=45
  twistMotion=Locked
 }
 {bone1Name=dyn_canister_r, bone2Name=spine_01, swing1LimitDegrees=50, swing2LimitDegrees=20, twistLimitDegrees=45, twistMotion=Locked}
 {
  bone1Name=dyn_strap_01
  bone2Name=dyn_satchel_l
  swing1LimitDegrees=45
  swing2LimitDegrees=45
  swing2Motion=Locked
  twistLimitDegrees=45
  twistMotion=Locked
 }
 {
  bone1Name=dyn_strap_02
  bone2Name=dyn_satchel_l
  swing1LimitDegrees=100
  swing2LimitDegrees=45
  swing2Motion=Locked
  twistLimitDegrees=45
  twistMotion=Locked
 }
]
