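# Animation Blueprint for the Mr Ice Guy male player body, adding secondary motion to its pickaxe, satchel and coat.
# It copies the attached parent's pose through a Control Rig, then Anim Dynamics and two RigidBody nodes.
# Additive pickaxe offsets and a post-dynamics Control Rig finish the pose; satchel and coat offset nodes are unconnected.
# The update event sets an unused crouch-walking flag and resets dynamics based on the player's water-jump state.
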AnimBlueprint /BRCosmetics/Characters/Player/Male/Medium/Bodies/M_MED_Mr_Ice_Guy/Meshes/M_MED_Mr_Ice_Guy_AnimBlueprint
target_skeleton=&/M_MED_Mr_Ice_Guy_Skeleton
graphs [
 AnimGraph {nodes [
   "Transform (Modify) Bone - Bone: dyn_satchel_l"!(ComponentPose, bAlphaBoolEnabled:bool=True, Translation="6,-8,5", Rotation="0,-40,0", Scale="1,1,1") -> (Pose) {
    alphaInputType=Bool
    rotationMode=BMM_Additive
    translationMode=BMM_Additive
    alphaBoolBlend {blendInTime=0.2, blendOption=Linear, blendOutTime=0.2}
    boneToModify {boneName=dyn_satchel_l}
    rotation {yaw=-40}
    translation=<6 -8 5>
   }
   "Transform (Modify) Bone - Bone: dyn_master_coat"!(ComponentPose, bAlphaBoolEnabled:bool=True, Translation="7,-5,2", Rotation="0,0,0", Scale="1,1,1") -> (Pose) {
    alphaInputType=Bool
    rotationMode=BMM_Additive
    translationMode=BMM_Additive
    alphaBoolBlend {blendInTime=0.2, blendOption=Linear, blendOutTime=0.2}
    boneToModify {boneName=dyn_master_coat}
    translation=<7 -5 2>
   }
   "Transform (Modify) Bone - Bone: dyn_coattail_l"!(ComponentPose, bAlphaBoolEnabled:bool=True, Translation="0,0,0", Rotation="20,-20,-20", Scale="1,1,1") -> (Pose) {
    alphaInputType=Bool
    rotationMode=BMM_Additive
    translationMode=BMM_Additive
    alphaBoolBlend {blendInTime=0.2, blendOption=Linear, blendOutTime=0.2}
    alphaScaleBiasClamp {bInterpResult=true}
    boneToModify {boneName=dyn_coattail_l}
    rotation=<rot 20 -20 -20>
   }
   "Transform (Modify) Bone - Bone: dyn_master_coattail_l"!(ComponentPose, bAlphaBoolEnabled:bool=True, Translation="9.5,-2,8", Rotation="-15,-90,-8.5", Scale="1,1,1") -> (Pose) {
    alphaInputType=Bool
    rotationMode=BMM_Additive
    translationMode=BMM_Additive
    alphaBoolBlend {blendInTime=0.2, blendOption=Linear, blendOutTime=0.2}
    alphaScaleBiasClamp {bInterpResult=true}
    boneToModify {boneName=dyn_master_coattail_l}
    rotation=<rot -15 -90 -8.5>
    translation=<9.5 -2 8>
   }
   n1 = Control Rig!(Source=Use cached pose 'CRCache_postDynamics'(), Alpha, Is crouching:bool=false) -> (Pose) {
    alphaBoolBlend {blendOption=Linear}
    controlRigAssetReference {blueprintRigClass=&/M_MED_Mr_Ice_Guy_after_dyn_CtrlRig1.M_MED_Mr_Ice_Guy_after_dyn_CtrlRig1_C}
   }
   Output Pose(Result=@n1) {layerGroup=None}
   n2 = Control Rig(Source=Copy Pose From Mesh() {bUseAttachedParent=true}, Is crouching=false) {
    alphaBoolBlend {blendOption=Linear}
    controlRigAssetReference {blueprintRigClass=&/M_MED_Mr_Ice_Guy_CtrlRig1.M_MED_Mr_Ice_Guy_CtrlRig1_C}
   }
   Save cached pose 'CRCache_inputPose'(Pose=@n2) {cacheName=CRCache_inputPose}
   n3 = Anim Dynamics - Bone: dyn_pickaxe(ComponentPose=Local To Component(LocalPose=Use cached pose 'CRCache_inputPose'()), Alpha=1, ExternalForce="0,0,0") {
    angularDampingOverride=0.95
    angularSpringConstant=0.5
    bGravityOverrideInSimSpace=true
    bOverrideAngularDamping=true
    bUseGravityOverride=true
    simulationSpace=RootRelative
    alphaBoolBlend {blendOption=Linear}
    boundBone {boneName=dyn_pickaxe}
    componentLinearAccScale=<0.02 0.02 0.02>
    gravityOverride {z=-980}
    physicsBodyDefinitions [{
      sphereCollisionRadius=0
      boundBone {boneName=dyn_pickaxe}
      boxExtents=<25 10 20>
      constraintSetup {angularLimitsMax=<5 30 5>, angularLimitsMin=<-5 -30 -5>, angularTarget=<0 0 0>}
      localJointOffset=<2 1 0>
     }]
    relativeSpaceBone {boneName=root}
   }
   n4 = Anim Dynamics - Bone: dyn_satchel_l(ComponentPose=@n3, Alpha=1, ExternalForce="0,0,0") {
    bGravityOverrideInSimSpace=true
    bUseGravityOverride=true
    simulationSpace=BoneRelative
    alphaBoolBlend {blendOption=Linear}
    boundBone {boneName=dyn_satchel_l}
    gravityOverride {z=-980}
    physicsBodyDefinitions [{
      sphereCollisionRadius=0
      boundBone {boneName=dyn_satchel_l}
      boxExtents=<20 15 15>
      constraintSetup {angularLimitsMax=<20 5 10>, angularLimitsMin=<-20 -5 -10>, angularTarget=<0 0 0>}
      localJointOffset=<0 0 0>
     }]
    relativeSpaceBone {boneName=root}
   }
   n5 = RigidBody(ComponentPose=@n4) {
    bClampLinearTranslationLimitToRefPose=true
    bOverrideWorldGravity=true
    overridePhysicsAsset=&/M_MED_Mr_Ice_Guy_Physics
    simulationSpace=BaseBoneSpace
    alphaBoolBlend {blendOption=Linear}
    baseBoneRef {boneName=root}
    componentLinearAccScale=<0.2 0.2 0.2>
    overrideWorldGravity {z=-980}
   }
   n6 = RigidBody(ComponentPose=@n5) {
    bOverrideWorldGravity=true
    overridePhysicsAsset=&/M_MED_Mr_Ice_Guy_Jacket_Physics1
    simulationSpace=BaseBoneSpace
    alphaBoolBlend {blendOption=Linear}
    baseBoneRef {boneName=root}
    componentAppliedLinearAccClamp=<3000 3000 3000>
    componentLinearAccScale=<0.2 0.2 0.2>
    overrideWorldGravity {z=-980}
    simSpaceSettings {dampingAlpha=0.5, maxAngularAcceleration=100, maxAngularVelocity=50, maxLinearAcceleration=1000, maxLinearVelocity=700}
   }
   Save cached pose 'CRCache_dynamics'(Pose=Component To Local(ComponentPose=@n6)) {cacheName=CRCache_dynamics}
   n7 = "Transform (Modify) Bone - Bone: dyn_pickaxe"!(
    ComponentPose=Local To Component(LocalPose=Use cached pose 'CRCache_dynamics'())
    bAlphaBoolEnabled:bool=True
    Translation="-1,-4.5,1"
    Rotation="-13,-45,20"
    Scale="1,1,1"
   ) -> (Pose) {
    alphaInputType=Bool
    rotationMode=BMM_Additive
    rotationSpace=BCS_ParentBoneSpace
    translationMode=BMM_Additive
    translationSpace=BCS_ParentBoneSpace
    alphaBoolBlend {blendInTime=0.2, blendOption=Linear, blendOutTime=0.2}
    boneToModify {boneName=dyn_pickaxe}
    rotation=<rot -13 -45 20>
    translation=<-1 -4.5 1>
   }
   n8 = "Transform (Modify) Bone - Bone: dyn_pickaxe"(ComponentPose=@n7, bAlphaBoolEnabled=True, Translation="0,0,0", Rotation="9,-2.5,3.5", Scale="1,1,1") {
    alphaInputType=Bool
    rotationMode=BMM_Additive
    translationMode=BMM_Additive
    alphaBoolBlend {blendInTime=0.2, blendOption=Linear, blendOutTime=0.2}
    boneToModify {boneName=dyn_pickaxe}
    rotation=<rot 9 -2.5 3.5>
   }
   Save cached pose 'CRCache_postDynamics'(Pose=Component To Local(ComponentPose=@n8)) {cacheName=CRCache_postDynamics}
  ]}
 EventGraph {nodes [
   Event BlueprintUpdateAnimation()
   > Set crouchWalking(crouchWalking:bool=AND Boolean(A=Get bIsCrouching(), B=Get bIsMoving2D())) -> (Output_Get:bool)
   > Cast To FortPlayerAnimInstance(Object=GetAnimInstance(self=GetOwnerSkeletalMeshComponent()))
   > Branch(Condition=Get bIsWaterJump() -> (bIsWaterJump:bool))
   > ResetDynamics(InTeleportType=TeleportPhysics)
   > ResetDynamics(InTeleportType=ResetPhysics)
  ]}
]
variables [crouchWalking {type=boolean}]
