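# Physics asset for the face accessory mesh of the Mr Ice Guy medium male player body, covering its dynamic hair and earring bones.
# A capsule on the head bone anchors small capsule bodies on ten hair bones and one earring bone.
# Each dynamic body is constrained directly to the head rather than chained, mostly with uniform swing and twist limits.
# The tenth hair bone differs by locking its first swing axis.
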PhysicsAsset /BRCosmetics/Characters/Player/Male/Medium/Bodies/M_MED_Mr_Ice_Guy/Meshes/Parts/M_MED_Mr_Ice_Guy_FaceAcc_PhysicsAsset
bodies {
 dyn_earing [{shapeType=Capsule, length=0.5, radius=0.5}]
 dyn_hair1 [{shapeType=Capsule, length=2, radius=2, center=<0.05 0 0>, rotation=<rot -90 0 0>}]
 dyn_hair10 [{shapeType=Capsule, length=2, radius=1.5, rotation=<rot -90 0 0>}]
 dyn_hair2 [{shapeType=Capsule, length=3, radius=2, center=<0.1 0 0>, rotation=<rot -90 0 0>}]
 dyn_hair3 [{shapeType=Capsule, length=0.8, radius=0.8, rotation=<rot -90 0 0>}]
 dyn_hair4 [{shapeType=Capsule, length=2, radius=2, rotation=<rot -90 0 0>}]
 dyn_hair5 [{shapeType=Capsule, length=2, radius=2, rotation=<rot -90 0 0>}]
 dyn_hair6 [{shapeType=Capsule, length=2, radius=2, rotation=<rot -90 0 0>}]
 dyn_hair7 [{shapeType=Capsule, length=2, radius=2, rotation=<rot -90 0 0>}]
 dyn_hair8 [{shapeType=Capsule, length=4, radius=1, rotation=<rot -90 0 0>}]
 dyn_hair9 [{shapeType=Capsule, length=2, radius=2, rotation=<rot -90 0 0>}]
 head [{shapeType=Capsule, length=9, radius=5.5, center=<2.1637 2.7743 0>, rotation=<rot 90 0 0>}]
}
constraints [
 {bone1Name=dyn_hair1, bone2Name=head, swing1LimitDegrees=45, swing2LimitDegrees=45, twistLimitDegrees=45}
 {bone1Name=dyn_hair2, bone2Name=head, swing1LimitDegrees=45, swing2LimitDegrees=45, twistLimitDegrees=45}
 {bone1Name=dyn_hair3, bone2Name=head, swing1LimitDegrees=45, swing2LimitDegrees=45, twistLimitDegrees=45}
 {bone1Name=dyn_hair4, bone2Name=head, swing1LimitDegrees=45, swing2LimitDegrees=45, twistLimitDegrees=45}
 {bone1Name=dyn_hair5, bone2Name=head, swing1LimitDegrees=45, swing2LimitDegrees=45, twistLimitDegrees=45}
 {bone1Name=dyn_hair6, bone2Name=head, swing1LimitDegrees=45, swing2LimitDegrees=45, twistLimitDegrees=45}
 {bone1Name=dyn_hair7, bone2Name=head, swing1LimitDegrees=45, swing2LimitDegrees=45, twistLimitDegrees=45}
 {bone1Name=dyn_hair8, bone2Name=head, swing1LimitDegrees=45, swing2LimitDegrees=45, twistLimitDegrees=45}
 {bone1Name=dyn_hair9, bone2Name=head, swing1LimitDegrees=45, swing2LimitDegrees=45, twistLimitDegrees=45}
 {bone1Name=dyn_hair10, bone2Name=head, swing1LimitDegrees=5, swing1Motion=Locked, swing2LimitDegrees=45, twistLimitDegrees=45}
 {bone1Name=dyn_earing, bone2Name=head, swing1LimitDegrees=45, swing2LimitDegrees=45, twistLimitDegrees=45}
]
